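; Gameplay ability for the Battle Pass S21 Week 1 narrative quest that plays a repeating world effect through gameplay cues.
; On activation from an event, it stores the event's instigator and starts a five-second looping timer.
; Each loop executes the configured GC Tag burst cue locally on that instigator at three hard-coded world locations, separated by short random delays to stagger the effects.
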(event Ability|EventActivateAbilityFromEvent
  (params (EventData :type Gameplay%20Event%20Data%20Structure%20%28by%20ref%29))
  (bind (_eventtag _instigator) (Utilities|Struct|BreakGameplayEventData EventData))
  (|CallParentFunction EventData)
  (Variables|Default|SetInstigator _instigator)
  (Utilities|Time|SetTimerbyEvent (delegate Custom|FX%20Loop) 5.0 true))

(event Custom|FX%20Loop
  (bind _gc_tag (Variables|Default|GetGC%20Tag))
  (bind _gameplaycuetag (Utilities|Struct|BreakGameplayCueTag _gc_tag))
  (bind _instigator (Variables|Default|GetInstigator))
  (GameplayCue|ExecuteGameplayCue%28Burst%29Local _instigator _gameplaycuetag (Ability|GameplayCue|MakeGameplayCueParameters :NormalizedMagnitude 0.0 :RawMagnitude 0.0 :Location "41309.000000,-74575.000000,4273.000000" :Normal "0,0,0" :GameplayEffectLevel 1 :AbilityLevel 1 :bReplicateLocationWhenUsingMinimalRepProxy false))
  (Utilities|FlowControl|Delay (Math|Random|RandomFloatinRange 0.2 0.5))
  (GameplayCue|ExecuteGameplayCue%28Burst%29Local _instigator _gameplaycuetag (Ability|GameplayCue|MakeGameplayCueParameters :NormalizedMagnitude 0.0 :RawMagnitude 0.0 :Location "38881.000000,-75025.000000,5527.000000" :Normal "0,0,0" :GameplayEffectLevel 1 :AbilityLevel 1 :bReplicateLocationWhenUsingMinimalRepProxy false))
  (Utilities|FlowControl|Delay (Math|Random|RandomFloatinRange 0.2 0.5))
  (GameplayCue|ExecuteGameplayCue%28Burst%29Local _instigator _gameplaycuetag (Ability|GameplayCue|MakeGameplayCueParameters :NormalizedMagnitude 0.0 :RawMagnitude 0.0 :Location "36689.000000,-74807.000000,4992.000000" :Normal "0,0,0" :GameplayEffectLevel 1 :AbilityLevel 1 :bReplicateLocationWhenUsingMinimalRepProxy false)))
